ConversationDatabase /BattlePassS37_Quests/Conversations/Story/Conversation_S37_Story_P01
entry_points [
 {tag=Conversation.Entry.Season37.Story.P01.Q02, destinations [#1]}
 {tag=Conversation.Entry.Season37.Story.P01.Q05, destinations [#2]}
 {tag=Conversation.Entry.Season37.Story.P01.Q05B, destinations [#3]}
]
nodes [
 ConversationEntryPointNode_3 {
  id=#3
  class=ConversationEntryPointNode
  outputs [#4]
  properties {entryTag {tagName=Conversation.Entry.Season37.Story.P01.Q05B}}
 }
 Speech_C_11 {
  id=#4
  class=Speech_C
  outputs [#5]
  properties {
   softTaskIcon=&T_UI_NPC_MainQuest
   generalConfig {defaultMessage=...Anybody else kind of craving calamari right now?, defaultMetadataParameters [RageDebris {value=""}]}
  }
  sub_nodes [
   ConversationChoiceNode_6 {
    class=ConversationChoiceNode
    properties {
     defaultChoiceDisplayText=Tentacle
     choiceTags {gameplayTags [{tagName=Conversation.Choice.Category.MainQuestline}, {tagName=Conversation.Choice.AutoSelectIfValid}]}
    }
   }
   FortConversationRequirement_ControllerRequirement_3 {
    class=FortConversationRequirement_ControllerRequirement
    properties {
     nodeName=Require: Has Completed Quest
     requirement {
      class=FortControllerRequirement_HasCompletedQuest
      properties {bInverseCondition=true, questDefinition=/BattlePassS37_Quests/Items/QuestItems/Story/P01/quest_S37_story_p01_q05}
     }
    }
   }
   FortConversationRequirement_ControllerRequirement_4 {
    class=FortConversationRequirement_ControllerRequirement
    properties {
     nodeName=Require: Is Quest Active
     requirement {
      class=FortControllerRequirement_IsQuestActive
      properties {questDefinition=&quest_S37_story_p01_q05}
     }
    }
   }
   FortConversationRequirement_ControllerRequirement_5 {
    class=FortConversationRequirement_ControllerRequirement
    properties {
     nodeName=Require: Has Completed Quest Objective
     requirement {
      class=FortControllerRequirement_HasCompletedQuestObjective
      properties {
       bInverseCondition=true
       objectiveBackendName=quest_S37_story_p01_q05_obj0
       questDefinition=&quest_S37_story_p01_q05
      }
     }
    }
   }
  ]
 }
 Speech_C_8 {
  id=#5
  class=Speech_C
  outputs [#6]
  properties {generalConfig {defaultMessage=The tentacle and the bugs are able to get to us from the Spirit Realm without using a portal.}}
  sub_nodes [ConversationChoiceNode_4 {
    class=ConversationChoiceNode
    properties {defaultChoiceDisplayText=Continue, choiceTags {gameplayTags [{tagName=Athena.Quests.S37.Conversations.Story.01.05.B}]}}
   }]
 }
 Speech_C_9 {
  id=#6
  class=Speech_C
  outputs [#7]
  properties {generalConfig {defaultMessage="We're trying to figure out exactly how they do that. And where. "}}
  sub_nodes [ConversationChoiceNode_1 {class=ConversationChoiceNode, properties {defaultChoiceDisplayText=Continue}}]
 }
 Speech_C_6 {
  id=#7
  class=Speech_C
  outputs []
  properties {generalConfig {defaultMessage="Care to help us out, pal?"}}
  sub_nodes [ConversationChoiceNode_5 {class=ConversationChoiceNode, properties {defaultChoiceDisplayText=Continue}}]
 }
 ConversationEntryPointNode_2 {
  id=#2
  class=ConversationEntryPointNode
  outputs [#8]
  properties {entryTag {tagName=Conversation.Entry.Season37.Story.P01.Q05}}
 }
 Speech_C_5 {
  id=#8
  class=Speech_C
  outputs [#9]
  properties {
   softTaskIcon=&T_UI_NPC_MainQuest
   generalConfig {defaultMessage="So... the big tentacle. First of all, try to ignore the smell. ", defaultMetadataParameters [RageDebris {value=""}]}
  }
  sub_nodes [
   ConversationChoiceNode_17 {
    class=ConversationChoiceNode
    properties {
     defaultChoiceDisplayText=Tentacle
     choiceTags {gameplayTags [{tagName=Conversation.Choice.Category.MainQuestline}, {tagName=Conversation.Choice.AutoSelectIfValid}]}
    }
   }
   FortConversationRequirement_ControllerRequirement_15 {
    class=FortConversationRequirement_ControllerRequirement
    properties {
     nodeName=Require: Has Completed Quest
     requirement {
      class=FortControllerRequirement_HasCompletedQuest
      properties {bInverseCondition=true, questDefinition=&quest_S37_story_p01_q05}
     }
    }
   }
   FortConversationRequirement_ControllerRequirement_16 {
    class=FortConversationRequirement_ControllerRequirement
    properties {
     nodeName=Require: Is Quest Active
     requirement {
      class=FortControllerRequirement_IsQuestActive
      properties {questDefinition=&quest_S37_story_p01_q05}
     }
    }
   }
   FortConversationRequirement_ControllerRequirement_17 {
    class=FortConversationRequirement_ControllerRequirement
    properties {
     nodeName=Require: Has Completed Quest Objective
     requirement {
      class=FortControllerRequirement_HasCompletedQuestObjective
      properties {
       bInverseCondition=true
       objectiveBackendName=quest_S37_story_p01_q05_obj0
       questDefinition=&quest_S37_story_p01_q05
      }
     }
    }
   }
  ]
 }
 Speech_C_2 {
  id=#9
  class=Speech_C
  outputs [#10]
  properties {generalConfig {defaultMessage=We know the tentacle and the Swarm are able to access the Island from the Spirit Realm.}}
  sub_nodes [ConversationChoiceNode_15 {
    class=ConversationChoiceNode
    properties {defaultChoiceDisplayText=Continue, choiceTags {gameplayTags [{tagName=Athena.Quests.S37.Conversations.Story.01.05.A}]}}
   }]
 }
 Speech_C_4 {
  id=#10
  class=Speech_C
  outputs [#11]
  properties {generalConfig {defaultMessage="We're studying exactly how they do that, but we need more data. Like the location of their access point."}}
  sub_nodes [ConversationChoiceNode_0 {class=ConversationChoiceNode, properties {defaultChoiceDisplayText=Continue}}]
 }
 Speech_C_3 {
  id=#11
  class=Speech_C
  outputs []
  properties {generalConfig {defaultMessage="And probably more stinky specimens. Something to look forward to, right?"}}
  sub_nodes [ConversationChoiceNode_16 {class=ConversationChoiceNode, properties {defaultChoiceDisplayText=Continue}}]
 }
 ConversationEntryPointNode_1 {
  id=#1
  class=ConversationEntryPointNode
  outputs [#12]
  properties {entryTag {tagName=Conversation.Entry.Season37.Story.P01.Q02}}
 }
 Speech_C_10 {
  id=#12
  class=Speech_C
  outputs [#13]
  properties {
   softTaskIcon=&T_UI_NPC_MainQuest
   generalConfig {
    defaultMessage=Name's General Onyx Winter. I'm leading O.X.R. in our valiant fight against the Swarm.
    defaultMetadataParameters [RageDebris {value=""}]
   }
  }
  sub_nodes [
   ConversationChoiceNode_14 {
    class=ConversationChoiceNode
    properties {defaultChoiceDisplayText=Get Briefed, choiceTags {gameplayTags [{tagName=Conversation.Choice.Category.MainQuestline}]}}
   }
   FortConversationRequirement_ControllerRequirement_12 {
    class=FortConversationRequirement_ControllerRequirement
    properties {
     nodeName=Require: Has Completed Quest
     requirement {
      class=FortControllerRequirement_HasCompletedQuest
      properties {bInverseCondition=true, questDefinition=/BattlePassS37_Quests/Items/QuestItems/Story/P01/quest_S37_story_p01_q02}
     }
    }
   }
   FortConversationRequirement_ControllerRequirement_13 {
    class=FortConversationRequirement_ControllerRequirement
    properties {
     nodeName=Require: Is Quest Active
     requirement {
      class=FortControllerRequirement_IsQuestActive
      properties {questDefinition=&quest_S37_story_p01_q02}
     }
    }
   }
   FortConversationRequirement_ControllerRequirement_14 {
    class=FortConversationRequirement_ControllerRequirement
    properties {
     nodeName=Require: Has Completed Quest Objective
     requirement {
      class=FortControllerRequirement_HasCompletedQuestObjective
      properties {
       bInverseCondition=true
       objectiveBackendName=quest_S37_story_p01_q02_obj0
       questDefinition=&quest_S37_story_p01_q02
      }
     }
    }
   }
  ]
 }
 Speech_C_0 {
  id=#13
  class=Speech_C
  outputs [#14]
  properties {generalConfig {defaultMessage=Nice swagger on the battlefield. You're kinda like a young Ripp Slade.}}
  sub_nodes [ConversationChoiceNode_12 {
    class=ConversationChoiceNode
    properties {defaultChoiceDisplayText=Continue, choiceTags {gameplayTags [{tagName=Athena.Quests.S37.Conversations.Story.01.02}]}}
   }]
 }
 Speech_C_1 {
  id=#14
  class=Speech_C
  outputs []
  properties {generalConfig {defaultMessage="We could use a little more of that right now. So if you're interested in joining up, consider this your formal welcome."}}
  sub_nodes [ConversationChoiceNode_13 {class=ConversationChoiceNode, properties {defaultChoiceDisplayText=Continue}}]
 }
]
