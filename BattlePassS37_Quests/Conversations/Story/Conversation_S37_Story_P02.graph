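# Conversation database for Season 37 story part P02, quest 05: an NPC receives eggs and theorizes they could open access to the Spirit Realm.
# One entry tag leads into a linear chain of four speech nodes.
# The auto-selected main-questline "Give Eggs" choice requires the quest to be active, with both the quest and objective 0 uncompleted.
# The two middle Continue choices carry the quest's conversation tag; the last speech ends the chain.
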ConversationDatabase /BattlePassS37_Quests/Conversations/Story/Conversation_S37_Story_P02
entry_points [{tag=Conversation.Entry.Season37.Story.P02.Q05, destinations [#1]}]
nodes [
 ConversationEntryPointNode_2 {
  id=#1
  class=ConversationEntryPointNode
  outputs [#2]
  properties {entryTag {tagName=Conversation.Entry.Season37.Story.P02.Q05}}
 }
 Speech_C_5 {
  id=#2
  class=Speech_C
  outputs [#3]
  properties {
   softTaskIcon=&T_UI_NPC_Conversation_Give
   generalConfig {
    defaultMessage="If my theory's correct, these eggs are Queen eggs, laid in the Spirit Realm and brought here."
    defaultMetadataParameters [RageDebris {value=""}]
   }
  }
  sub_nodes [
   ConversationChoiceNode_8 {
    class=ConversationChoiceNode
    properties {
     defaultChoiceDisplayText=Give Eggs
     choiceTags {gameplayTags [{tagName=Conversation.Choice.Category.MainQuestline}, {tagName=Conversation.Choice.AutoSelectIfValid}]}
    }
   }
   FortConversationRequirement_ControllerRequirement_6 {
    class=FortConversationRequirement_ControllerRequirement
    properties {
     nodeName=Require: Has Completed Quest
     requirement {
      class=FortControllerRequirement_HasCompletedQuest
      properties {bInverseCondition=true, questDefinition=/BattlePassS37_Quests/Items/QuestItems/Story/P02/quest_S37_story_p02_q05}
     }
    }
   }
   FortConversationRequirement_ControllerRequirement_7 {
    class=FortConversationRequirement_ControllerRequirement
    properties {
     nodeName=Require: Is Quest Active
     requirement {
      class=FortControllerRequirement_IsQuestActive
      properties {questDefinition=&quest_S37_story_p02_q05}
     }
    }
   }
   FortConversationRequirement_ControllerRequirement_8 {
    class=FortConversationRequirement_ControllerRequirement
    properties {
     nodeName=Require: Has Completed Quest Objective
     requirement {
      class=FortControllerRequirement_HasCompletedQuestObjective
      properties {
       bInverseCondition=true
       objectiveBackendName=quest_S37_story_p02_q05_obj0
       questDefinition=&quest_S37_story_p02_q05
      }
     }
    }
   }
  ]
 }
 Speech_C_0 {
  id=#3
  class=Speech_C
  outputs [#4]
  properties {generalConfig {defaultMessage="If that's true, we can use their energy signature to weaken the Swarm's access point to the Island."}}
  sub_nodes [ConversationChoiceNode_3 {
    class=ConversationChoiceNode
    properties {defaultChoiceDisplayText=Continue, choiceTags {gameplayTags [{tagName=Athena.Quests.S37.Conversations.Story.02.05}]}}
   }]
 }
 Speech_C_2 {
  id=#4
  class=Speech_C
  outputs [#5]
  properties {generalConfig {defaultMessage="And if we're *really* lucky, I can reverse-engineer some way to use the access point too."}}
  sub_nodes [ConversationChoiceNode_0 {
    class=ConversationChoiceNode
    properties {defaultChoiceDisplayText=Continue, choiceTags {gameplayTags [{tagName=Athena.Quests.S37.Conversations.Story.02.05}]}}
   }]
 }
 Speech_C_4 {
  id=#5
  class=Speech_C
  outputs []
  properties {generalConfig {defaultMessage=This egg might hold the key to accessing the Spirit Realm without a portal. And winning once and for all.}}
  sub_nodes [ConversationChoiceNode_7 {class=ConversationChoiceNode, properties {defaultChoiceDisplayText=Continue}}]
 }
]
