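# Conversation database for Season 37 story part P03 NPC quest briefings about the Splinter Blades and entering the Spirit Realm.
# Four entry tags start linear Speech chains: three for Q01 (A, B, C) and one for Q06.
# Each opening line requires its quest active and uncompleted with a specific objective unfinished; second-line choices carry quest conversation tags.
# The Q06 gate checks objective quest_S37_story_p01_q02_obj0, whose name differs from its quest.
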ConversationDatabase /BattlePassS37_Quests/Conversations/Story/Conversation_S37_Story_P03
entry_points [
 {tag=Conversation.Entry.Season37.Story.P03.Q06, destinations [#1]}
 {tag=Conversation.Entry.Season37.Story.P03.Q01.A, destinations [#2]}
 {tag=Conversation.Entry.Season37.Story.P03.Q01.B, destinations [#3]}
 {tag=Conversation.Entry.Season37.Story.P03.Q01.C, destinations [#4]}
]
nodes [
 ConversationEntryPointNode_0 {
  id=#4
  class=ConversationEntryPointNode
  outputs [#5]
  properties {entryTag {tagName=Conversation.Entry.Season37.Story.P03.Q01.C}}
 }
 Speech_C_7 {
  id=#5
  class=Speech_C
  outputs [#6]
  properties {
   softTaskIcon=&T_UI_NPC_MainQuest
   generalConfig {defaultMessage="You've come for the daggers. I know. ", defaultMetadataParameters [RageDebris {value=""}]}
  }
  sub_nodes [
   ConversationChoiceNode_6 {
    class=ConversationChoiceNode
    properties {
     defaultChoiceDisplayText=Get Briefed
     choiceTags {gameplayTags [{tagName=Conversation.Choice.Category.MainQuestline}, {tagName=Conversation.Choice.AutoSelectIfValid}]}
    }
   }
   FortConversationRequirement_ControllerRequirement_3 {
    class=FortConversationRequirement_ControllerRequirement
    properties {
     nodeName=Require: Has Completed Quest
     requirement {
      class=FortControllerRequirement_HasCompletedQuest
      properties {bInverseCondition=true, questDefinition=/BattlePassS37_Quests/Items/QuestItems/Story/P03/quest_S37_story_p03_q01}
     }
    }
   }
   FortConversationRequirement_ControllerRequirement_4 {
    class=FortConversationRequirement_ControllerRequirement
    properties {
     nodeName=Require: Is Quest Active
     requirement {
      class=FortControllerRequirement_IsQuestActive
      properties {questDefinition=&quest_S37_story_p03_q01}
     }
    }
   }
   FortConversationRequirement_ControllerRequirement_5 {
    class=FortConversationRequirement_ControllerRequirement
    properties {
     nodeName=Require: Has Completed Quest Objective
     requirement {
      class=FortControllerRequirement_HasCompletedQuestObjective
      properties {
       bInverseCondition=true
       objectiveBackendName=quest_S37_story_p03_q01_obj2
       questDefinition=&quest_S37_story_p03_q01
      }
     }
    }
   }
  ]
 }
 Speech_C_3 {
  id=#6
  class=Speech_C
  outputs [#7]
  properties {generalConfig {defaultMessage=I used to think I'd use these Splinter Blades to save the Island. Now... I know that isn't my fate.}}
  sub_nodes [ConversationChoiceNode_4 {
    class=ConversationChoiceNode
    properties {defaultChoiceDisplayText=Continue, choiceTags {gameplayTags [{tagName=Athena.Quests.S37.Conversations.Story.03.01.C}]}}
   }]
 }
 Speech_C_6 {
  id=#7
  class=Speech_C
  outputs []
  properties {generalConfig {defaultMessage="...Take them. Do what I failed to do. Cut through the veil, and strike down this demonic incursion for all of us. "}}
  sub_nodes [ConversationChoiceNode_5 {class=ConversationChoiceNode, properties {defaultChoiceDisplayText=Continue}}]
 }
 ConversationEntryPointNode_3 {
  id=#3
  class=ConversationEntryPointNode
  outputs [#8]
  properties {entryTag {tagName=Conversation.Entry.Season37.Story.P03.Q01.B}}
 }
 Speech_C_11 {
  id=#8
  class=Speech_C
  outputs [#9]
  properties {
   softTaskIcon=&T_UI_NPC_MainQuest
   generalConfig {
    defaultMessage=Something truly awful is coming to the Island. We can rise to meet it... or wait for whatever it brings.
    defaultMetadataParameters [RageDebris {value=""}]
   }
  }
  sub_nodes [
   ConversationChoiceNode_2 {
    class=ConversationChoiceNode
    properties {
     defaultChoiceDisplayText=Get Briefed
     choiceTags {gameplayTags [{tagName=Conversation.Choice.Category.MainQuestline}, {tagName=Conversation.Choice.AutoSelectIfValid}]}
    }
   }
   FortConversationRequirement_ControllerRequirement_0 {
    class=FortConversationRequirement_ControllerRequirement
    properties {
     nodeName=Require: Has Completed Quest
     requirement {
      class=FortControllerRequirement_HasCompletedQuest
      properties {bInverseCondition=true, questDefinition=&quest_S37_story_p03_q01}
     }
    }
   }
   FortConversationRequirement_ControllerRequirement_1 {
    class=FortConversationRequirement_ControllerRequirement
    properties {
     nodeName=Require: Is Quest Active
     requirement {
      class=FortControllerRequirement_IsQuestActive
      properties {questDefinition=&quest_S37_story_p03_q01}
     }
    }
   }
   FortConversationRequirement_ControllerRequirement_2 {
    class=FortConversationRequirement_ControllerRequirement
    properties {
     nodeName=Require: Has Completed Quest Objective
     requirement {
      class=FortControllerRequirement_HasCompletedQuestObjective
      properties {
       bInverseCondition=true
       objectiveBackendName=quest_S37_story_p03_q01_obj1
       questDefinition=&quest_S37_story_p03_q01
      }
     }
    }
   }
  ]
 }
 Speech_C_8 {
  id=#9
  class=Speech_C
  outputs [#10]
  properties {generalConfig {defaultMessage=Fighting the Dark Presence with the Splinter Blades feels a little like attacking a Great Turtle with a toothpick...}}
  sub_nodes [ConversationChoiceNode_0 {
    class=ConversationChoiceNode
    properties {defaultChoiceDisplayText=Continue, choiceTags {gameplayTags [{tagName=Athena.Quests.S37.Conversations.Story.03.01.B}]}}
   }]
 }
 Speech_C_9 {
  id=#10
  class=Speech_C
  outputs []
  properties {generalConfig {defaultMessage="But maybe we'll be lucky. Every creature has some sort of soft underbelly, right?"}}
  sub_nodes [ConversationChoiceNode_1 {class=ConversationChoiceNode, properties {defaultChoiceDisplayText=Continue}}]
 }
 ConversationEntryPointNode_2 {
  id=#2
  class=ConversationEntryPointNode
  outputs [#11]
  properties {entryTag {tagName=Conversation.Entry.Season37.Story.P03.Q01.A}}
 }
 Speech_C_5 {
  id=#11
  class=Speech_C
  outputs [#12]
  properties {
   softTaskIcon=&T_UI_NPC_MainQuest
   generalConfig {
    defaultMessage="Yes, I was aware that Hope and Jones had cooked up another reckless, ramshackle plan."
    defaultMetadataParameters [RageDebris {value=""}]
   }
  }
  sub_nodes [
   ConversationChoiceNode_17 {
    class=ConversationChoiceNode
    properties {
     defaultChoiceDisplayText=Get Briefed
     choiceTags {gameplayTags [{tagName=Conversation.Choice.Category.MainQuestline}, {tagName=Conversation.Choice.AutoSelectIfValid}]}
    }
   }
   FortConversationRequirement_ControllerRequirement_15 {
    class=FortConversationRequirement_ControllerRequirement
    properties {
     nodeName=Require: Has Completed Quest
     requirement {
      class=FortControllerRequirement_HasCompletedQuest
      properties {bInverseCondition=true, questDefinition=&quest_S37_story_p03_q01}
     }
    }
   }
   FortConversationRequirement_ControllerRequirement_16 {
    class=FortConversationRequirement_ControllerRequirement
    properties {
     nodeName=Require: Is Quest Active
     requirement {
      class=FortControllerRequirement_IsQuestActive
      properties {questDefinition=&quest_S37_story_p03_q01}
     }
    }
   }
   FortConversationRequirement_ControllerRequirement_17 {
    class=FortConversationRequirement_ControllerRequirement
    properties {
     nodeName=Require: Has Completed Quest Objective
     requirement {
      class=FortControllerRequirement_HasCompletedQuestObjective
      properties {
       bInverseCondition=true
       objectiveBackendName=quest_S37_story_p03_q01_obj0
       questDefinition=&quest_S37_story_p03_q01
      }
     }
    }
   }
  ]
 }
 Speech_C_2 {
  id=#12
  class=Speech_C
  outputs [#13]
  properties {generalConfig {defaultMessage="Entering the Spirit Realm is, to put it mildly, ill-advised."}}
  sub_nodes [ConversationChoiceNode_15 {
    class=ConversationChoiceNode
    properties {defaultChoiceDisplayText=Continue, choiceTags {gameplayTags [{tagName=Athena.Quests.S37.Conversations.Story.03.01.A}]}}
   }]
 }
 Speech_C_4 {
  id=#13
  class=Speech_C
  outputs [#14]
  properties {generalConfig {defaultMessage="However, we have a dearth of remaining options. So if we must enter the Spirit Realm to survive..."}}
  sub_nodes [ConversationChoiceNode_3 {class=ConversationChoiceNode, properties {defaultChoiceDisplayText=Continue}}]
 }
 Speech_C_12 {
  id=#14
  class=Speech_C
  outputs []
  properties {generalConfig {defaultMessage=*Sigh*... I suppose it's better than the alternative.}}
  sub_nodes [ConversationChoiceNode_16 {class=ConversationChoiceNode, properties {defaultChoiceDisplayText=Continue}}]
 }
 ConversationEntryPointNode_1 {
  id=#1
  class=ConversationEntryPointNode
  outputs [#15]
  properties {entryTag {tagName=Conversation.Entry.Season37.Story.P03.Q06}}
 }
 Speech_C_10 {
  id=#15
  class=Speech_C
  outputs [#16]
  properties {
   softTaskIcon=&T_UI_NPC_MainQuest
   generalConfig {defaultMessage=The veil between realms is weak. And we have what we need to cross over., defaultMetadataParameters [RageDebris {value=""}]}
  }
  sub_nodes [
   ConversationChoiceNode_14 {
    class=ConversationChoiceNode
    properties {
     defaultChoiceDisplayText=Get Briefed
     choiceTags {gameplayTags [{tagName=Conversation.Choice.Category.MainQuestline}, {tagName=Conversation.Choice.AutoSelectIfValid}]}
    }
   }
   FortConversationRequirement_ControllerRequirement_12 {
    class=FortConversationRequirement_ControllerRequirement
    properties {
     nodeName=Require: Has Completed Quest
     requirement {
      class=FortControllerRequirement_HasCompletedQuest
      properties {bInverseCondition=true, questDefinition=/BattlePassS37_Quests/Items/QuestItems/Story/P03/quest_S37_story_p03_q06}
     }
    }
   }
   FortConversationRequirement_ControllerRequirement_13 {
    class=FortConversationRequirement_ControllerRequirement
    properties {
     nodeName=Require: Is Quest Active
     requirement {
      class=FortControllerRequirement_IsQuestActive
      properties {questDefinition=&quest_S37_story_p03_q06}
     }
    }
   }
   FortConversationRequirement_ControllerRequirement_14 {
    class=FortConversationRequirement_ControllerRequirement
    properties {
     nodeName=Require: Has Completed Quest Objective
     requirement {
      class=FortControllerRequirement_HasCompletedQuestObjective
      properties {
       bInverseCondition=true
       objectiveBackendName=quest_S37_story_p01_q02_obj0
       questDefinition=&quest_S37_story_p03_q06
      }
     }
    }
   }
  ]
 }
 Speech_C_0 {
  id=#16
  class=Speech_C
  outputs [#17]
  properties {generalConfig {defaultMessage="We have the Splinter Blades, and the old Swarm tunnels, and a small army of battle-tested allies. "}}
  sub_nodes [ConversationChoiceNode_12 {
    class=ConversationChoiceNode
    properties {defaultChoiceDisplayText=Continue, choiceTags {gameplayTags [{tagName=Athena.Quests.S37.Conversations.Story.03.06}]}}
   }]
 }
 Speech_C_1 {
  id=#17
  class=Speech_C
  outputs [#18]
  properties {generalConfig {defaultMessage="We've been in worse situations before, so let's do what we always do. "}}
  sub_nodes [ConversationChoiceNode_7 {class=ConversationChoiceNode, properties {defaultChoiceDisplayText=Continue}}]
 }
 Speech_C_13 {
  id=#18
  class=Speech_C
  outputs []
  properties {generalConfig {defaultMessage="Stay focused, work together, and make it count."}}
  sub_nodes [ConversationChoiceNode_13 {class=ConversationChoiceNode, properties {defaultChoiceDisplayText=Continue}}]
 }
]
