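# Season 37 conversation database for the VO Device, letting players toggle voice-over and sound effects during events.
# Table entry greets and links to HasDevice and NoDevice branches; NPC entries offer similar branches directly.
# Requirements on the VODevice.Enabled tag gate Enable versus Disable, and each offers Accept or Refuse, ending in GrantReaction.
# Every Accept shares one quest choice tag, and Disable confirmations also say VO will be enabled.
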ConversationDatabase /BattlePassS37_Quests/Conversations/Story/Conversation_S37_VODevice
entry_points [
 {tag=Conversation.Entry.Season37.VODevice.NPC.NoDevice, destinations [#1]}
 {tag=Conversation.Entry.Season37.VODevice.NPC.HasDevice, destinations [#2]}
 {tag=Conversation.Entry.Season37.VODevice.Table.HasDevice, destinations [#3]}
 {tag=Conversation.Entry.Season37.VODevice.Table.NoDevice, destinations [#4]}
 {tag=Conversation.Entry.Season37.VODevice.Table, destinations [#5]}
]
nodes [
 ConversationEntryPointNode_15 {
  id=#5
  class=ConversationEntryPointNode
  outputs [#6]
  properties {entryTag {tagName=Conversation.Entry.Season37.VODevice.Table}}
 }
 Speech_C_39 {
  id=#6
  class=Speech_C
  outputs [#7, #8]
  properties {nodeName="", generalConfig {defaultMessage="[Welcome to O.X.R. Communication Services!]"}}
 }
 ConversationLinkNode_2 {
  id=#8
  class=ConversationLinkNode
  properties {remoteEntryTag {tagName=Conversation.Entry.Season37.VODevice.Table.HasDevice}}
 }
 ConversationLinkNode_0 {
  id=#7
  class=ConversationLinkNode
  properties {remoteEntryTag {tagName=Conversation.Entry.Season37.VODevice.Table.NoDevice}}
 }
 ConversationEntryPointNode_16 {
  id=#4
  class=ConversationEntryPointNode
  outputs [#9]
  properties {entryTag {tagName=Conversation.Entry.Season37.VODevice.Table.NoDevice}}
 }
 Speech_C_35 {
  id=#9
  class=Speech_C
  outputs [#10, #11]
  properties {
   softTaskIcon=/Game/Athena/HUD/Interaction/T_UI_PayPhone
   generalConfig {
    defaultMessage="[Accepting this will enable Voice Over and sound effects during certain gameplay events]"
    defaultMetadataParameters [RageDebris {value=""}]
   }
  }
  sub_nodes [
   ConversationChoiceNode_54 {
    class=ConversationChoiceNode
    properties {defaultChoiceDisplayText=Enable VO, choiceTags {gameplayTags [{tagName=Conversation.Choice.Category.MainQuestline}]}}
   }
   FortConversationRequirement_ControllerRequirement_56 {
    class=FortConversationRequirement_ControllerRequirement
    properties {
     nodeName=Require: Has GameplayTag
     requirement {class=FortControllerRequirement_HasOwnedTag, properties {bInverseCondition=true, ownedTag {tagName=Athena.Quest.VODevice.Enabled}}}
    }
   }
  ]
 }
 Speech_C_34 {
  id=#11
  class=Speech_C
  outputs [#12]
  properties {softTaskIcon=/Game/Athena/HUD/Art/squadArt/T_UI_NPCCom_Dismiss, generalConfig {defaultMessage=(VO will stay disabled during certain events)}}
  sub_nodes [ConversationChoiceNode_53 {class=ConversationChoiceNode, properties {defaultChoiceDisplayText=Refuse}}]
 }
 FortConversationTaskNode_GrantReaction_5 {id=#12, class=FortConversationTaskNode_GrantReaction, outputs []}
 Speech_C_33 {
  id=#10
  class=Speech_C
  outputs [#13]
  properties {
   softTaskIcon=/Game/Athena/HUD/Interaction/T_UI_NPC_Confirm
   generalConfig {defaultMessage=(Voice Over and sound effects will now be enabled during certain events)}
  }
  sub_nodes [ConversationChoiceNode_52 {
    class=ConversationChoiceNode
    properties {defaultChoiceDisplayText=Accept, choiceTags {gameplayTags [{tagName=Athena.Quests.S37.Conversations.Story.01.02}]}}
   }]
 }
 FortConversationTaskNode_GrantReaction_2 {id=#13, class=FortConversationTaskNode_GrantReaction, outputs []}
 ConversationEntryPointNode_14 {
  id=#3
  class=ConversationEntryPointNode
  outputs [#14]
  properties {entryTag {tagName=Conversation.Entry.Season37.VODevice.Table.HasDevice}}
 }
 Speech_C_38 {
  id=#14
  class=Speech_C
  outputs [#15, #16]
  properties {
   softTaskIcon=&T_UI_PayPhone
   generalConfig {defaultMessage="[Do you want to stop receiving audio communication during Events?]", defaultMetadataParameters [RageDebris {value=""}]}
  }
  sub_nodes [
   ConversationChoiceNode_57 {
    class=ConversationChoiceNode
    properties {defaultChoiceDisplayText=Disable VO, choiceTags {gameplayTags [{tagName=Conversation.Choice.Category.MainQuestline}]}}
   }
   FortConversationRequirement_ControllerRequirement_57 {
    class=FortConversationRequirement_ControllerRequirement
    properties {
     nodeName=Require: Has GameplayTag
     requirement {class=FortControllerRequirement_HasOwnedTag, properties {ownedTag {tagName=Athena.Quest.VODevice.Enabled}}}
    }
   }
  ]
 }
 Speech_C_37 {
  id=#16
  class=Speech_C
  outputs [#17]
  properties {
   softTaskIcon=&T_UI_NPCCom_Dismiss
   generalConfig {defaultMessage=(This will keep VO disabled during certain events)}
  }
  sub_nodes [ConversationChoiceNode_56 {class=ConversationChoiceNode, properties {defaultChoiceDisplayText=Refuse}}]
 }
 FortConversationTaskNode_GrantReaction_7 {id=#17, class=FortConversationTaskNode_GrantReaction, outputs []}
 Speech_C_36 {
  id=#15
  class=Speech_C
  outputs [#18]
  properties {softTaskIcon=&T_UI_NPC_Confirm, generalConfig {defaultMessage=(This will enable VO during certain Events)}}
  sub_nodes [ConversationChoiceNode_55 {
    class=ConversationChoiceNode
    properties {defaultChoiceDisplayText=Accept, choiceTags {gameplayTags [{tagName=Athena.Quests.S37.Conversations.Story.01.02}]}}
   }]
 }
 FortConversationTaskNode_GrantReaction_6 {id=#18, class=FortConversationTaskNode_GrantReaction, outputs []}
 ConversationEntryPointNode_0 {
  id=#2
  class=ConversationEntryPointNode
  outputs [#19]
  properties {entryTag {tagName=Conversation.Entry.Season37.VODevice.NPC.HasDevice}}
 }
 Speech_C_27 {
  id=#19
  class=Speech_C
  outputs [#20, #21]
  properties {
   softTaskIcon=&T_UI_PayPhone
   generalConfig {defaultMessage="[Do you want to stop receiving audio communication during Events?]", defaultMetadataParameters [RageDebris {value=""}]}
  }
  sub_nodes [
   ConversationChoiceNode_79 {
    class=ConversationChoiceNode
    properties {defaultChoiceDisplayText=Disable VO, choiceTags {gameplayTags [{tagName=Conversation.Choice.Category.MainQuestline}]}}
   }
   FortConversationRequirement_ControllerRequirement_32 {
    class=FortConversationRequirement_ControllerRequirement
    properties {
     nodeName=Require: Has GameplayTag
     requirement {class=FortControllerRequirement_HasOwnedTag, properties {ownedTag {tagName=Athena.Quest.VODevice.Enabled}}}
    }
   }
  ]
 }
 Speech_C_28 {
  id=#21
  class=Speech_C
  outputs [#22]
  properties {
   softTaskIcon=&T_UI_NPCCom_Dismiss
   generalConfig {defaultMessage=(This will keep VO disabled during certain events)}
  }
  sub_nodes [ConversationChoiceNode_80 {class=ConversationChoiceNode, properties {defaultChoiceDisplayText=Refuse}}]
 }
 FortConversationTaskNode_GrantReaction_23 {id=#22, class=FortConversationTaskNode_GrantReaction, outputs []}
 Speech_C_29 {
  id=#20
  class=Speech_C
  outputs [#23]
  properties {softTaskIcon=&T_UI_NPC_Confirm, generalConfig {defaultMessage=(This will enable VO during certain Events)}}
  sub_nodes [ConversationChoiceNode_81 {
    class=ConversationChoiceNode
    properties {defaultChoiceDisplayText=Accept, choiceTags {gameplayTags [{tagName=Athena.Quests.S37.Conversations.Story.01.02}]}}
   }]
 }
 FortConversationTaskNode_GrantReaction_24 {id=#23, class=FortConversationTaskNode_GrantReaction, outputs []}
 ConversationEntryPointNode_1 {
  id=#1
  class=ConversationEntryPointNode
  outputs [#24]
  properties {entryTag {tagName=Conversation.Entry.Season37.VODevice.NPC.NoDevice}}
 }
 Speech_C_30 {
  id=#24
  class=Speech_C
  outputs [#25, #26]
  properties {
   softTaskIcon=&T_UI_PayPhone
   generalConfig {defaultMessage="[Do you want to receive audio communication during Events?]", defaultMetadataParameters [RageDebris {value=""}]}
  }
  sub_nodes [
   ConversationChoiceNode_82 {
    class=ConversationChoiceNode
    properties {
     defaultChoiceDisplayText=Enable VO
     choiceTags {gameplayTags [{tagName=Conversation.Entry.Season37.VODevice.NPC.NoDevice}, {tagName=Conversation.Choice.Category.MainQuestline}]}
    }
   }
   FortConversationRequirement_ControllerRequirement_33 {
    class=FortConversationRequirement_ControllerRequirement
    properties {
     nodeName=Require: Has GameplayTag
     requirement {class=FortControllerRequirement_HasOwnedTag, properties {bInverseCondition=true, ownedTag {tagName=Athena.Quest.VODevice.Enabled}}}
    }
   }
  ]
 }
 Speech_C_31 {
  id=#26
  class=Speech_C
  outputs [#27]
  properties {
   softTaskIcon=&T_UI_NPCCom_Dismiss
   generalConfig {defaultMessage=(This will keep VO disabled during certain events)}
  }
  sub_nodes [ConversationChoiceNode_83 {class=ConversationChoiceNode, properties {defaultChoiceDisplayText=Refuse}}]
 }
 FortConversationTaskNode_GrantReaction_25 {id=#27, class=FortConversationTaskNode_GrantReaction, outputs []}
 Speech_C_32 {
  id=#25
  class=Speech_C
  outputs [#28]
  properties {softTaskIcon=&T_UI_NPC_Confirm, generalConfig {defaultMessage=(This will enable VO during certain Events)}}
  sub_nodes [ConversationChoiceNode_84 {
    class=ConversationChoiceNode
    properties {defaultChoiceDisplayText=Accept, choiceTags {gameplayTags [{tagName=Athena.Quests.S37.Conversations.Story.01.02}]}}
   }]
 }
 FortConversationTaskNode_GrantReaction_26 {id=#28, class=FortConversationTaskNode_GrantReaction, outputs []}
]
